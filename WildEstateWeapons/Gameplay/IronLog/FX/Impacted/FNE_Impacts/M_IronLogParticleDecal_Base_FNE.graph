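# Translucent unlit Niagara sprite and mesh particle material that draws the IronLog impact decal.
# The decal texture's green channel feeds an emissive-adjustment function and, times particle alpha and an inverted depth fade, forms opacity, keeping it near intersecting surfaces.
# A quality switch adds a second, adjustable depth fade on High, Epic and non-mobile default paths.
# World-position projection math is present but unconnected; sampling uses mesh UVs.
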Material /WildEstateWeapons/Gameplay/IronLog/FX/Impacted/FNE_Impacts/M_IronLogParticleDecal_Base_FNE
graph {
 properties {
  bUsedWithNiagaraMeshParticles=true
  bUsedWithNiagaraSprites=true
  blendMode=BLEND_Translucent
  pixelDepthOffsetMode=PDOM_Legacy
  shadingModel=MSM_Unlit
  usedShadingModels=Unlit
 }
 nodes [
  n1 = MaterialFunctionCall(Vector 2=ParticleSize()) {materialFunction=/Engine/Functions/Engine_MaterialFunctions02/Utility/VectorLength}
  n2 = Divide(
   A=Subtract(
    A=ObjectPositionWS()
    B=MaterialFunctionCall() {materialFunction=&WorldPositionBehindTranslucency}.Result
   )
   B=@n1.V2 Length
  )
  n3 = Constant() {r=0.5}
  ComponentMask(None=Subtract(A=Add(A=@n2, B=@n3), B=Divide(A=@n3, B=@n1.V2 Length) {desc=Scale the center offset by the scaling factor}) {desc=Re offset to maintain center}) {b=true, r=true}
  n4 = TextureSample(UVs=TextureCoordinate()) {texture=/WildEstateWeapons/Gameplay/IronLog/FX/Decals/FNE_Decals/T_IronLogDecal_FNE}
  MaterialFunctionCall(Emissive=@n4.G) {materialFunction=&MF_FX_EmissiveAdjustments}
  ParticlePositionWS()
  n5 = Constant() {r=50}
  n6 = Multiply(A=Multiply(A=ParticleColor().A, B=@n4.G), B=OneMinus(None=DepthFade(FadeDistance=@n5)))
  n7 = DepthFade(Opacity=@n6, FadeDistance=ScalarParameter() {defaultValue=5, parameterName=DepthFadeDistance})
  QualitySwitch(Default=ShadingPathSwitch(Default=@n7, Mobile=@n6), Low=@n6, High=@n7, Medium=@n6, Epic=@n7)
 ]
}
